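# Material function that builds tileable cylindrical projection UVs around local Z, with separate DDX and DDY outputs for texture sampling.
# U comes from a radial angle and V from inverted height within the pre-skinned local bounds, using vertex-interpolated pre-skinned position.
# Derivatives come from a continuous position offset along the camera vector and scaled, bypassing the angular wrap.
# Inputs are Tile and DistortionAdjustment.
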MaterialFunction /Game/Global/Materials/MaterialGym/Materials/MaterialElements/UVs/CylindricalProjection/MF_CylindricalProjUV
graph {nodes [
  FunctionOutput(None=NamedRerouteUsage()) {outputName=UVs}
  FunctionOutput(None=NamedRerouteUsage()) {outputName=DDX, sortPriority=1}
  FunctionOutput(None=NamedRerouteUsage()) {outputName=DDY, sortPriority=2}
  n1 = VertexInterpolator(VS=LocalPosition() {localOrigin=InstancePreSkinning}.XYZ)
  n2 = CameraVectorWS()
  n3 = Multiply(
   A=Transform(None=Multiply(A=@n2) {constB=-1}) {transformSourceType=TRANSFORMSOURCE_World, transformType=TRANSFORM_Local}
   B=NamedRerouteUsage()
  ) {constB=100}
  n4 = Divide(A=NamedRerouteDeclaration(None=Add(A=@n1.PS, B=@n3)) {name=SL_PreskinnedLocalPos, nodeColor=<rgba 0 1 0.0859 1>}, B=NamedRerouteUsage())
  n5 = Normalize(VectorInput=Transform(None=Constant3Vector() {constant=<rgba 1 0 0 0>}) {transformSourceType=TRANSFORMSOURCE_View})
  NamedRerouteDeclaration(None=DDX(Value=DotProduct(A=@n4, B=@n5))) {name=DDX, nodeColor=<rgba 0 1 0.9062 1>}
  NamedRerouteDeclaration(None=DDY(Value=ComponentMask(None=@n4) {b=true})) {name=DDY, nodeColor=<rgba 1 0 0.9609 1>}
  n6 = Multiply(A=MaterialFunctionCall(Vector or UVs=NamedRerouteUsage()) {materialFunction=&VectorToRadialValue}.Vector Converted to Angle) {constB=-1}
  n7 = ComponentMask(None=NamedRerouteUsage()) {r=true}
  n8 = ComponentMask(None=NamedRerouteUsage()) {g=true}
  n9 = PreSkinnedLocalBounds()
  n10 = Subtract(A=ComponentMask(None=@n1.PS) {b=true}, B=ComponentMask(None=@n9.Min) {b=true})
  n11 = NamedRerouteDeclaration(None=ComponentMask(None=@n9.Extents) {b=true}) {name=SL_PreskinnedLocalBoundsZ, nodeColor=<rgba 0 1 0.3203 1>}
  n12 = OneMinus(None=Saturate(None=Divide(A=@n10, B=@n11)))
  NamedRerouteDeclaration(None=AppendVector(A=Multiply(A=@n6, B=@n7) {constB=2}, B=Multiply(A=@n8, B=@n12))) {name=UVs, nodeColor=<rgba 0.6875 0 1 1>}
  n13 = FunctionInput() {bUsePreviewValueAsDefault=true, inputName=Tile, inputType=FunctionInput_Vector2, previewValue=<1 1 0 1>}
  NamedRerouteDeclaration(None=@n13) {name=Tile, nodeColor=<rgba 0.4531 0 1 1>}
  n14 = FunctionInput() {bUsePreviewValueAsDefault=true, inputName=DistortionAdjustment, inputType=FunctionInput_Scalar, sortPriority=1, previewValue=<-5 0 0 1>}
  NamedRerouteDeclaration(None=@n14) {name=DistortionAdjustment, nodeColor=<rgba 0.4531 0 1 1>}
 ]}
