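# Material Gym example material that demonstrates cylindrical UV projection using the UV editor color grid texture.
# The MF_CylindricalProjUV function, with its tile and distortion adjustment inputs left unwired, generates the UVs for the texture sample.
# The sample uses derivative mip mode, with the same UV output wired directly into both DDX and DDY.
# It is a Default Lit material flagged for skeletal mesh and mesh deformer use.
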Material /Game/Global/Materials/MaterialGym/Materials/MaterialElements/UVs/CylindricalProjection/M_CylindricalProjection
graph {
 properties {bAutomaticallySetUsageInEditor=false, bUsedWithMeshDeformer=true, bUsedWithSkeletalMesh=true, usedShadingModels=Default Lit}
 nodes [
  n1 = MaterialFunctionCall!(Tile, DistortionAdjustment) {materialFunction=&/MF_CylindricalProjUV}
  TextureSample!(UVs=@n1.UVs, Tex, "DDX(UVs)"=@n1.UVs, "DDY(UVs)"=@n1.UVs, Apply View MipBias) {mipValueMode=TMVM_Derivative, texture=/UVEditor/Textures/UVEditorColorGrid_Color}
 ]
}
