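# Example Material Gym material that demonstrates cylindrical projection UVs, flagged for skeletal mesh and mesh deformer use with Default Lit shading.
# MF_CylindricalProjUV_Mod generates the UVs, which feed both the animated MF_StormNoise function and the MF_Polar UV function.
# Two scalar parameters offset the polar center.
# No links from these functions to material outputs appear in the graph.
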Material /Game/Global/Materials/MaterialGym/Materials/MaterialElements/UVs/CylindricalProjection/M_CylindricalProjection_Example
graph {
 properties {bAutomaticallySetUsageInEditor=false, bUsedWithMeshDeformer=true, bUsedWithSkeletalMesh=true, usedShadingModels=Default Lit}
 nodes [
  Constant()
  n1 = MaterialFunctionCall!(Tile, DistortionAdjustment) {materialFunction=&/MF_CylindricalProjUV_Mod}
  MaterialFunctionCall!(
   UV=@n1.UVs
   NoiseNormal
   NoiseTexture01
   NoiseTexture02
   DistortionTexture
   NoiseColorLight
   NoiseColorDarkOffset
   NoiseTint
   NoiseTiling
   NoiseSpeed
   DistortionTiling
   DistortionSpeed
  ) {materialFunction=/Game/Global/Materials/MaterialGym/Materials/MaterialElements/UVs/AnimatedNoise/MF_StormNoise}
  MaterialFunctionCall!(
   UV_Input=@n1.UVs
   PolarUV_CenterOffsetG=ScalarParameter() {defaultValue=-0.26}
   PolarUV_CenterOffsetR=ScalarParameter() {defaultValue=0.26, parameterName=Param_1}
   PolarUV_LenghtScale
   PolarUV_RadialScale
   PolarUV_UseTwist
   PolarUV_TwistOffset
   PolarUVTwistSpeed
   PolarUV_TwistRadius
  ) {materialFunction=/Game/Global/Materials/MaterialGym/Materials/MaterialElements/UVs/UVMainpulation/PolarUV/MF_Polar}
 ]
}
