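# Test material in MaterialGym for cylindrical UV projection, with usage enabled for skeletal meshes and mesh deformers.
# MF_CylindricalProjUV_Mod takes a tiling parameter and outputs UVs, which are panned to sample a color-grid test texture.
# The sample uses explicit derivatives from the function, with its DDX output feeding both DDX and DDY.
# A static switch selects the function's ArmMask or a constant for visualization.
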Material /Game/Global/Materials/MaterialGym/Materials/MaterialElements/UVs/CylindricalProjection/M_CylindricalProjection_Mod
graph {
 properties {
  bAutomaticallySetUsageInEditor=false
  bHasPixelAnimation=true
  bUsedWithMeshDeformer=true
  bUsedWithSkeletalMesh=true
  usedShadingModels=Default Lit
 }
 nodes [
  n1 = MaterialFunctionCall!(Tile=ComponentMask(None=VectorParameter() {parameterName=UVTiling, defaultValue=<rgba 1 1 0 1>}.RGB) {g=true, r=true}, DistortionAdjustment) {materialFunction=&/MF_CylindricalProjUV_Mod}
  StaticSwitchParameter(True=@n1.ArmMask, False=Constant()) {group=Vis, parameterName=VisArmMask}
  n2 = VectorParameter() {parameterName=PanningSpeed, defaultValue=<rgba 0 0 0 1>}
  TextureSampleParameter2D!(
   UVs=Panner(Coordinate=@n1.UVs, Speed=ComponentMask(None=@n2.RGB) {g=true, r=true}) {speedY=0.1}
   "DDX(UVs)"=@n1.DDX
   "DDY(UVs)"=@n1.DDX
   Apply View MipBias
  ) {constMipValue=10, mipValueMode=TMVM_Derivative, parameterName=TestTexture, texture=/UVEditor/Textures/UVEditorColorGrid_Color}
 ]
}
